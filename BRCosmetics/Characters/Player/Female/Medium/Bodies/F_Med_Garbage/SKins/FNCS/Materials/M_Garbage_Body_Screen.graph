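# Character body material for the Garbage FNCS skin that adds an animated, glowing sci-fi screen between the shared 2023 character PreFX and PostFX functions.
# A custom node recenters and scales switchable UV0/UV1, using only the X center and scale for both axes.
# Two panning wave layers and a time-pulsed bar are tinted over a tiled two-color background.
# An area mask blacks out base color and swaps in this emissive.
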Material /BRCosmetics/Characters/Player/Female/Medium/Bodies/F_Med_Garbage/SKins/FNCS/Materials/M_Garbage_Body_Screen
graph {
 properties {
  bUseMaterialAttributes=true
  bUsedWithMeshDeformer=true
  bUsedWithMorphTargets=true
  bUsedWithSkeletalMesh=true
  pixelDepthOffsetMode=PDOM_Legacy
  shadingModel=MSM_FromMaterialExpression
  usedShadingModels=Default Lit | Subsurface | Clear Coat | Subsurface Profile | Cloth
 }
 nodes [
  n1 = MaterialFunctionCall() {materialFunction=&MF_CharacterParent_2023_PreFX}
  n2 = GetMaterialAttributes(None=@n1.Result) {attributeGetTypes [#1, #2]}
  n3 = NamedRerouteUsage()
  n4 = VectorParameter() {
   group=Scifi_Screen
   parameterName=SineTexture_Center01
   channelNames {a=ScaleY, b=ScaleX, g=CenterY, r=CenterX}
   defaultValue=<rgba 0.5 0.5 1 1>
  }
  MaterialExpressionAppendVector_19 = AppendVector(A=@n4.CenterX, B=@n4.CenterX)
  MaterialExpressionAppendVector_20 = AppendVector(A=@n4.ScaleX, B=@n4.ScaleX)
  MaterialExpressionStaticSwitchParameter_0 = StaticSwitchParameter(True=TextureCoordinate() {coordinateIndex=1}, False=TextureCoordinate()) {group=Scifi_Screen, parameterName=UseTexcoord01}
  n5 = Custom!(
   UVPoint=@MaterialExpressionAppendVector_19
   UVScale=@MaterialExpressionAppendVector_20
   OriginalUV=@MaterialExpressionStaticSwitchParameter_0
  ) {
   code="float2 DistUV =OriginalUV-UVPoint;\r\nDistUV = DistUV/UVScale + float2(0.5,0.5);\r\n\r\nreturn DistUV;"
   description=CopyToPointUV
   outputType=CMOT_Float2
   inputs [
    {
     inputName=UVPoint
     input {expression=&self:MaterialExpressionAppendVector_19}
    }
    {
     inputName=UVScale
     input {expression=&self:MaterialExpressionAppendVector_20}
    }
    {
     inputName=OriginalUV
     input {expression=&self:MaterialExpressionStaticSwitchParameter_0}
    }
   ]
  }
  n6 = ScalarParameter() {defaultValue=0.15, group=Scifi_Screen, parameterName=SpeedX01}
  n7 = ScalarParameter() {group=Scifi_Screen, parameterName=SpeedY01}
  n8 = TextureSampleParameter2D(UVs=Panner(Coordinate=@n5, Speed=AppendVector(A=@n6, B=@n7))) {
   group=Scifi_Screen
   parameterName=SineWave
   samplerSource=SSM_Wrap_WorldGroupSettings
   samplerType=SAMPLERTYPE_Masks
   texture=/BRCosmetics/Characters/Player/Female/Medium/Bodies/F_Med_Garbage/SKins/FNCS/Textures/T_Wave
  }
  n9 = Power(Base=@n8.R) {constExponent=4}
  n10 = Power(Base=@n8.G) {constExponent=6}
  n11 = ScalarParameter() {defaultValue=0.1, group=Scifi_Screen, parameterName=SpeedX02}
  n12 = ScalarParameter() {group=Scifi_Screen, parameterName=SpeedY02}
  n13 = TextureSampleParameter2D(UVs=Panner(Coordinate=@n5, Speed=AppendVector(A=@n11, B=@n12))) {
   group=Scifi_Screen
   parameterName=SineWave
   samplerSource=SSM_Wrap_WorldGroupSettings
   samplerType=SAMPLERTYPE_Masks
   texture=&T_Wave
  }
  n14 = Add(A=Power(Base=@n13.R) {constExponent=6}, B=Power(Base=@n13.G) {constExponent=6})
  n15 = TextureSampleParameter2D(UVs=@n5) {
   group=Scifi_Screen
   parameterName=SineWave
   samplerSource=SSM_Wrap_WorldGroupSettings
   samplerType=SAMPLERTYPE_Masks
   texture=&T_Wave
  }
  n16 = ScalarParameter() {defaultValue=0.1, group=Scifi_Screen, parameterName=BarIntensity}
  n17 = ConstantBiasScale(None=Sine(None=Time()))
  n18 = Add(
   A=Add(A=Add(A=@n9, B=@n10), B=Multiply(B=@n14) {constA=0.3})
   B=Multiply(A=Multiply(A=@n15.B, B=@n16), B=LinearInterpolate(Alpha=@n17) {constA=0.5, constB=2})
  )
  n19 = VectorParameter() {group=Scifi_Screen, parameterName=BasicEmissiveColor, defaultValue=<rgba 0 0.629 1 1>}
  n20 = TextureSampleParameter2D(UVs=Multiply(A=@n5, B=ScalarParameter() {defaultValue=3, group=Scifi_Screen, parameterName=BackGroundTilling})) {
   group=Scifi_Screen
   parameterName=BackGround
   samplerType=SAMPLERTYPE_Masks
   texture=/BRCosmetics/Characters/Player/Female/Medium/Bodies/F_Med_Garbage/SKins/FNCS/Textures/T_ScreenBackGround
  }
  n21 = Add(
   A=Multiply(A=Multiply(A=@n18, B=@n19.RGB), B=ScalarParameter() {defaultValue=3, parameterName=Emissive_Sine_Intensity})
   B=LinearInterpolate(
    A=VectorParameter() {group=Scifi_Screen, parameterName=BackGroundEmissiveColor_High, defaultValue=<rgba 0 0.0904 0.1437 1>}.RGB
    B=VectorParameter() {group=Scifi_Screen, parameterName=BackGroundEmissiveColor_Low, defaultValue=<rgba 0 0.0432 0.0688 1>}.RGB
    Alpha=@n20.RGB
   )
  )
  n22 = NamedRerouteUsage()
  MaterialFunctionCall(
   Material=SetMaterialAttributes!(
    MaterialAttributes=@n2.MaterialAttributes
    Base Color=LinearInterpolate(A=@n2.Base Color, Alpha=@n3) {constB=0}
    Emissive Color=LinearInterpolate(A=@n2.Emissive Color, B=@n21, Alpha=@n22) {constB=0}
   ) {attributeSetTypes [#1, #2]}
   M=@n1.M
   Pre Skinned Local Position=@n1.M
   Distance Blend=@n1.M
   Fresnel=@n1.M
  ) {materialFunction=&MF_CharacterParent_2023_PostFX}
  n23 = TextureSampleParameter2D() {
   group=Scifi_Screen
   parameterName=ScreeArea
   samplerType=SAMPLERTYPE_Masks
   texture=/BRCosmetics/Characters/Player/Female/Medium/Bodies/F_Med_Garbage/SKins/FNCS/Textures/T_ScreenMask_Refine
  }
  NamedRerouteDeclaration(None=@n23.R) {name=AreaMask, nodeColor=<rgba 0 0.7891 1 1>}
  TextureCoordinate()
 ]
}
